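# Parent cloth-shaded material for the ProngCling backpack, combining the 2023 anime character shader functions with animated flipbook art.
# Pre-skinning local-position thresholds mask face, hand, foot, body and hair regions to select cells in a 4x4 atlas on UV1.
# A sine pulse times the eyes, while offset curve-atlas timers drive mouth, hand and foot frames.
# The sample supplies lit and darkened shaded colors, with temporal responsiveness set for the changing frames.
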Material /BRCosmetics/Accessories/FORT_Backpacks/Backpack_F_MED_ProngCling/Materials/M_ProngCling_Backpack_Parent
graph {
 properties {
  bAutomaticallySetUsageInEditor=false
  bUseMaterialAttributes=true
  bUsedWithClothing=true
  bUsedWithMeshDeformer=true
  bUsedWithMorphTargets=true
  bUsedWithNiagaraMeshParticles=true
  bUsedWithSkeletalMesh=true
  bUsedWithStaticLighting=true
  pixelDepthOffsetMode=PDOM_Legacy
  previewMesh=/BRCosmetics/Accessories/FORT_Backpacks/Backpack_F_MED_ProngCling/Meshes/F_MED_ProngCling_Pack
  shadingModel=MSM_Cloth
  usedShadingModels=Cloth
 }
 nodes [
  n1 = MaterialFunctionCall!(CustomUV) {materialFunction=/Game/Global/Materials/Character_Parent_2023/Anime_Parent_2023/MaterialFunctions/MFC_AnimeShader_2023_Pre}
  n2 = MaterialFunctionCall!(CustomLightVector, OverrideNormal_WS, STT=@n1.STT, Normal_Map=@n1.STT, DFL=@n1.STT, FX=@n1.STT) {materialFunction=/Game/Global/Materials/Character_Parent_2023/Anime_Parent_2023/MaterialFunctions/MFC_AnimeShader_2023_Calculations}
  MaterialFunctionCall!(
   Shading_Gradient=@n2.Shading_Gradient
   Threshold_Gradient=@n2.Shading_Gradient
   Specular_Gradient=@n2.Shading_Gradient
   Sheen_Mask=@n2.Shading_Gradient
   Thickness=@n2.Shading_Gradient
   DFL_Mask=@n2.Shading_Gradient
   FX=@n2.Shading_Gradient
   Metallic_Gradient=@n2.Shading_Gradient
   RimLight_Gradient=@n2.Shading_Gradient
   NormalBuffer=@n2.Shading_Gradient
   --------
   CINE_LightSources=@n2.Shading_Gradient
   CINE_LightColor=@n2.Shading_Gradient
   ---------
   Color_Lit=NamedRerouteUsage()
   Color_Shaded=NamedRerouteUsage()
   SRM=@n1.SRM
   Emissive=@n1.SRM
   DFL_Color=@n1.SRM
   Specular_Color=@n1.SRM
   RimLight_Color=@n1.SRM
  ) {materialFunction=/Game/Global/Materials/Character_Parent_2023/Anime_Parent_2023/MaterialFunctions/MFC_AnimeShader_2023_Post}
  NamedRerouteDeclaration(None=@n1.Color_Lit) {name=IN_Color_Lit, nodeColor=<rgba 1 0.4922 0 1>}
  NamedRerouteDeclaration(None=@n1.Color_Shaded) {name=IN_Color_Shaded, nodeColor=<rgba 0.1719 0 1 1>}
  n3 = Constant() {r=4}
  n4 = AppendVector(A=@n3, B=@n3)
  n5 = Sine(None=Multiply(A=Time(), B=ScalarParameter() {defaultValue=0.25, group=ProngCling Backpack, parameterName=Speed_Eye}) {constB=0.25})
  n6 = Saturate(None=Subtract(A=Multiply(A=@n5) {constB=20}) {constB=19})
  n7 = Add(A=Time()) {constB=0.2356}
  n8 = ScalarParameter() {defaultValue=0.2, group=ProngCling Backpack, parameterName=Speed_Mouth}
  n9 = CurveAtlasRowParameter(CurveTime=Frac(None=Multiply(A=@n7, B=@n8) {constB=0.25})) {
   atlas=/BRCosmetics/Accessories/FORT_Backpacks/Backpack_F_MED_ProngCling/Textures/CA_ProngCling_Backpack_Flipbook
   curve=/BRCosmetics/Accessories/FORT_Backpacks/Backpack_F_MED_ProngCling/Textures/C_ProngCling_Backpack_Flipbook
   parameterName=FlipbookAnim
  }
  n10 = Add(A=Time()) {constB=0.6324}
  n11 = ScalarParameter() {defaultValue=0.5, group=ProngCling Backpack, parameterName=Speed_Hand}
  n12 = CurveAtlasRowParameter(CurveTime=Frac(None=Multiply(A=@n10, B=@n11) {constB=0.25})) {
   atlas=&CA_ProngCling_Backpack_Flipbook
   curve=&C_ProngCling_Backpack_Flipbook
   parameterName=FlipbookAnim
  }
  n13 = LinearInterpolate(
   A=LinearInterpolate(A=LinearInterpolate(Alpha=@n6) {constB=3}, B=LinearInterpolate(Alpha=@n9.G) {constA=4, constB=7}, Alpha=NamedRerouteUsage())
   B=LinearInterpolate(Alpha=@n12.B) {constA=8, constB=12}
   Alpha=NamedRerouteUsage()
  )
  n14 = Add(A=Time()) {constB=0.8435}
  n15 = ScalarParameter() {defaultValue=0.45, group=ProngCling Backpack, parameterName=Speed_Foot}
  n16 = CurveAtlasRowParameter(CurveTime=Frac(None=Multiply(A=@n14, B=@n15) {constB=0.25})) {
   atlas=&CA_ProngCling_Backpack_Flipbook
   curve=&C_ProngCling_Backpack_Flipbook
   parameterName=FlipbookAnim
  }
  n17 = LinearInterpolate(Alpha=@n16.B) {constA=12, constB=15}
  n18 = NamedRerouteUsage()
  n19 = LinearInterpolate(A=LinearInterpolate(A=LinearInterpolate(A=@n13, B=@n17, Alpha=@n18), Alpha=NamedRerouteUsage()) {constB=7}, Alpha=NamedRerouteUsage()) {constB=3}
  n20 = Floor(None=Fmod(A=@n19, B=@n3))
  n21 = Floor(None=Divide(A=@n19, B=@n3))
  n22 = Add(A=Divide(A=TextureCoordinate() {coordinateIndex=1}, B=@n4), B=Divide(A=AppendVector(A=@n20, B=@n21), B=@n4))
  n23 = TextureSample(
   UVs=@n22
   Tex=RecordTextureStreamingInfo!(
    Tex=TextureObject() {texture=/BRCosmetics/Accessories/FORT_Backpacks/Backpack_F_MED_ProngCling/Textures/T_ProngCling_Backpack_Flipbook}
    UVs=@n22
   ).Tex
  )
  NamedRerouteDeclaration(None=@n23.RGB) {name=OUT_ColorLit, nodeColor=<rgba 1 0.0938 0 1>}
  MaterialExpressionVertexInterpolator_0 = VertexInterpolator(VS=LocalPosition() {localOrigin=InstancePreSkinning}.XYZ)
  n24 = Convert(None=@MaterialExpressionVertexInterpolator_0.PS) {
   nodeName=Break Float3
   convertInputs [{
     type=Vector3
     expressionInput {
      expression=&self:MaterialExpressionVertexInterpolator_0
      inputName=None
      mask=0
      maskA=0
      maskB=0
      maskG=0
      maskR=0
      outputIndex=0
     }
    }]
   convertOutputs [{type=Scalar}, {type=Scalar}, {type=Scalar}]
  }
  NamedRerouteDeclaration(None=Step(Y=@n24) {constX=-23.7, constY=-14}) {name=Face_Eye, nodeColor=<rgba 0.5703 0 1 1>}
  NamedRerouteDeclaration(None=Multiply(
    A=@n23.RGB
    B=ScalarParameter() {defaultValue=0.5, group=01_Anime Maps, parameterName=Dynamic_ColorShaded_Darkening, sliderMax=1, sortPriority=11}
   )) {name=OUT_ColorShaded, nodeColor=<rgba 0 0.2266 1 1>}
  NamedRerouteDeclaration(None=Step(Y=@n24) {constX=-8.6}) {name=Face_Hand, nodeColor=<rgba 1 0 0.0703 1>}
  NamedRerouteDeclaration(None=Step(X=@n24) {constX=-8, constY=8.7}) {name=Face_Mouth, nodeColor=<rgba 0 0.3906 1 1>}
  NamedRerouteDeclaration(None=Step(Y=@n24) {constX=-8.05}) {name=Face_Foot, nodeColor=<rgba 0 0.4375 1 1>}
  NamedRerouteDeclaration(None=Step(X=@n24) {constX=-7, constY=-6.4}) {name=Face_Body, nodeColor=<rgba 0 0.8828 1 1>}
  NamedRerouteDeclaration(None=Step(X=@n24) {constY=9.25}) {name=Face_Hair, nodeColor=<rgba 1 0.7969 0 1>}
  TemporalResponsivenessOutput!(None=Constant() {r=1})
 ]
}
